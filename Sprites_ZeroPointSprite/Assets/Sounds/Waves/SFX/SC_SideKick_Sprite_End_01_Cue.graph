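# Sound cue for an end sound effect of the Zero Point Sprite sidekick.
# A single wave player at the root plays the SideKick_Sprite_End_01 wave, with no randomization or modulation.
# The cue belongs to the Fort_Gadgets sound class and overrides attenuation with natural-sound distance falloff.
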SoundCue /Sprites_ZeroPointSprite/Assets/Sounds/Waves/SFX/SC_SideKick_Sprite_End_01_Cue
root=WavePlayer_0
nodes [{id=WavePlayer_0, properties {soundWaveAssetPtr=&/SideKick_Sprite_End_01}}]
properties {
 bOverrideAttenuation=true
 maxDistance=4000
 soundClassObject=&Fort_Gadgets
 volumeMultiplier=1
 attenuationOverrides {distanceAlgorithm=NaturalSound}
}
